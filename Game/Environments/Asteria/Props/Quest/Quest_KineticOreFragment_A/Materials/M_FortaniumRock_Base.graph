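# Base material for the Fortanium rock of the Asteria Kinetic Ore Fragment quest prop, usable on skeletal meshes.
# Rock base color, normal and a specular mask, whose red channel drives both specular and roughness, feed MF_ActiveFortanium with a constant full-white mask.
# Fortnite_HitEffects, driven by custom primitive data for deformation, damage and cull fade, adds hit-and-shrink world position offset to that result.
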Material /Game/Environments/Asteria/Props/Quest/Quest_KineticOreFragment_A/Materials/M_FortaniumRock_Base
graph {
 properties {bUsedWithMeshDeformer=true, bUsedWithSkeletalMesh=true, pixelDepthOffsetMode=PDOM_Legacy, usedShadingModels=Default Lit}
 nodes [
  n1 = TextureSampleParameter2D() {parameterName=BaseColor, texture=/Game/Environments/Asteria/Props/Quest/Quest_KineticOreFragment_A/Textures/T_Fortanium_Rock_02_D}
  n2 = Constant()
  n3 = TextureSampleParameter2D() {
   parameterName=T_Specular
   samplerType=SAMPLERTYPE_Masks
   texture=/Game/Environments/Asteria/Props/Quest/Quest_KineticOreFragment_A/Textures/T_Fortanium_Rock_02_S
  }
  n4 = TextureSampleParameter2D() {
   parameterName=T_Normal
   samplerType=SAMPLERTYPE_Normal
   texture=/Game/Environments/Asteria/Props/Quest/Quest_KineticOreFragment_A/Textures/T_Fortanium_Rock_02_N
  }
  n5 = GetMaterialAttributes(None=MaterialFunctionCall!(
    Mask=Constant3Vector() {constant=<rgba 1 1 1 0>}
    Material=MakeMaterialAttributes!(
     BaseColor=@n1.RGB
     Metallic=@n2
     Specular=@n3.R
     Roughness=@n3.R
     Anisotropy
     EmissiveColor
     Opacity
     OpacityMask
     Normal=@n4.RGB
     Tangent
     WorldPositionOffset
     SubsurfaceColor
     ClearCoat
     ClearCoatRoughness
     AmbientOcclusion
     Refraction
     CustomizedUVs_0
     CustomizedUVs_1
     CustomizedUVs_2
     CustomizedUVs_3
     CustomizedUVs_4
     CustomizedUVs_5
     CustomizedUVs_6
     CustomizedUVs_7
     PixelDepthOffset
     ShadingModel
     Displacement
    )
   ) {materialFunction=&/MF_ActiveFortanium}.Result) {attributeGetTypes [#1]}
  n6 = MaterialFunctionCall!(
   WSDeformationCenter=VectorParameter() {bUseCustomPrimitiveData=true, group=z_code, parameterName=WSDeformationCenter, primitiveDataIndex=20}.RGB
   Code_DestructionAmount=ScalarParameter() {bUseCustomPrimitiveData=true, group=z_code, parameterName=Code_DestructionAmount, primitiveDataIndex=13}
   Code_UseUvForDamage
   Code_UseHitEffects=StaticBoolParameter() {defaultValue=true, group=z_code, parameterName=Code_UseHitEffects}
   Code_Emissive
   WSDamageRadius=ScalarParameter() {bUseCustomPrimitiveData=true, defaultValue=1, group=z_code, parameterName=WSDamageRadius, primitiveDataIndex=26}
   CrackDamageShader
   Diffuse
   ScalarModDamageCracks
   ScalarOffsetforDamageCracks
   ExposeLerpFunctionForInvertingTheHitDistanceCalc
   InvertDamageFalloff
   CrackWorldPosition
   WorldPosition
   DistanceCullFadeAnimationOffset=ScalarParameter() {bUseCustomPrimitiveData=true, group=z_code, parameterName=DistanceCullFadeAnimationOffset, primitiveDataIndex=12}
  ) {materialFunction=/Game/Packages/jl_tests/Fortnite_HitEffects}
  GetMaterialAttributes(None=SetMaterialAttributes!(MaterialAttributes=@n5.MaterialAttributes, World Position Offset=Add(A=@n5.World Position Offset, B=@n6.Hit And Shrink WPO)) {attributeSetTypes [#1]}) {attributeGetTypes [
    #2
    #3
    #4
    #5
    #6
    #7
    #1
   ]}
 ]
}
